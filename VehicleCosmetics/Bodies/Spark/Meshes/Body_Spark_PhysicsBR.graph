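# Physics asset for the Spark vehicle body cosmetic, named as its BR variant, providing collision for the skeletal mesh.
# It holds a single body entry named "body" with no listed shapes.
# It defines no constraints, so it describes no articulated joints or ragdoll chain.
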PhysicsAsset /VehicleCosmetics/Bodies/Spark/Meshes/Body_Spark_PhysicsBR
bodies {body []}
constraints []
